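; Stagger gameplay ability for the TidalCrane ToyFox boss that plays a left or right hit react montage at a hotfixable play rate.
; The montage side depends on whether the granting effect's instigator has the Melee Combo2 tag.
; A montage notify window applies and removes a disable-reposition gameplay effect, with application gated by a hotfix value.
; Ending the ability also clears that effect, by handle and by effect class.
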(event Ability|EventActivateAbility
  ;[0 0.15 0.03037 0.5] ActivateAbility: Play the appropriate hit react animation based on the attack swing.
  (|CallParentFunction)
  (Fortnite|FortShippingLog self "ActivateAbility: Invoked.")
  (bind (_asynctaskproxy _notifytag) (Ability|Tasks|PlayMontageAndWaitNotify "StaggerMontageTaskName" (select (GameplayTags|HasMatchingGameplayTag (Ability|GetAbilitySystemComponent (Ability|EffectContext|GetInstigatorActor (Ability|GetGrantedbyEffectContext))) "(TagName=\"Gameplay.Melee.Combo2\")") (Variables|Settings|GetHitReactMontageRight) (Variables|Settings|GetHitReactMontageLeft)) (ScalableFloat|GetCachedScalableFloatValue (Variables|Hotfix|GetHotfix_AnimationPlayRate)))
    (:OnCompleted
      _)
    (:OnBlendOut
      _)
    (:OnInterrupted
      _)
    (:OnCancelled
      _)
    (:OnNotifyBegin
      (if (== _notifytag "(TagName=\"TidalCrane.ToyFox.AnimNotify.Stagger.DisableReposition\")")
        (if (> (ScalableFloat|GetCachedScalableFloatValue (Variables|Hotfix|GetHotfix_ApplyDisableRepositionTag)) 0)
          (bind _returnvalue (Ability|ApplyGameplayEffectToOwner (Variables|Settings|GetDisableRepositionGameplayEffectClass)))
          (Variables|Transient|SetDisableRepositionGameplayEffectHandle _returnvalue))))
    (:OnNotifyEnd
      (if (== _notifytag "(TagName=\"TidalCrane.ToyFox.AnimNotify.Stagger.DisableReposition\")")
        (Ability|RemoveGameplayEffectFromOwnerWithHandle (Variables|Transient|GetDisableRepositionGameplayEffectHandle)))))
  (Ability|EndAbility))

(event Ability|EventOnEndAbility
  (params (bWasCancelled :type Boolean))
  (bind _returnvalue (Ability|GetAbilitySystemComponentfromActorInfo))
  ;[0.15 0.09603 0 0.5] OnEndAbility: Remove disable reposition gameplay tag.
  (Utilities|FlowControl|Sequence
    (:then_0
      (Fortnite|FortShippingLog self "OnEndAbility: Invoked.")
      (Ability|RemoveGameplayEffectFromOwnerWithHandle (Variables|Transient|GetDisableRepositionGameplayEffectHandle))
      (Utilities|IsValid _returnvalue
        (:Is%20Valid
          (GameplayEffects|RemoveActiveGameplayEffectbySourceEffect _returnvalue (Variables|Settings|GetDisableRepositionGameplayEffectClass) 0 -1))))
    (:then_1
      (|CallParentFunction bWasCancelled))))
